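# Sound cue for the spark foley of the Pottery Wheel emote.
# A random node picks one of six equally weighted spark waves, and a modulator adds slight pitch and volume variation.
# It uses emote foley class, attenuation and concurrency settings, keeps playing when silent, and has a submix send to the music audio-analysis submix.
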SoundCue /BRCosmetics/Sounds/Emotes/PotteryWheel/EmoteFoley_PotteryWheel_Sparks_Cue
root=Modulator_0
nodes [
 {id=Modulator_0, children [Random_0], properties {pitchMax=1.02, pitchMin=0.98, volumeMax=1.02, volumeMin=0.98}}
 {id=Random_0, children [WavePlayer_3,WavePlayer_4,WavePlayer_5,WavePlayer_6,WavePlayer_7,WavePlayer_8], properties {weights [1,1,1,1,1,1]}}
 {id=WavePlayer_3, properties {soundWaveAssetPtr=&/IP_Emote_PotteryWheel_Sparks_06}}
 {id=WavePlayer_4, properties {soundWaveAssetPtr=&/IP_Emote_PotteryWheel_Sparks_05}}
 {id=WavePlayer_5, properties {soundWaveAssetPtr=&/IP_Emote_PotteryWheel_Sparks_04}}
 {id=WavePlayer_6, properties {soundWaveAssetPtr=&/IP_Emote_PotteryWheel_Sparks_03}}
 {id=WavePlayer_7, properties {soundWaveAssetPtr=&/IP_Emote_PotteryWheel_Sparks_02}}
 {id=WavePlayer_8, properties {soundWaveAssetPtr=&/IP_Emote_PotteryWheel_Sparks_01}}
]
properties {
 attenuationSettings=&Fort_Emotes_Athena_Att
 audioPropertiesSheet=/BRCosmetics/Sounds/Settings/PropertySheets/Emotes/APS_Cosmetic_Emote_Foley
 maxDistance=1600
 soundClassObject=/Game/Sounds/FortSoundClassesAndModes/Fort_Foley_Emote
 virtualizationMode=PlayWhenSilent
 volumeMultiplier=1
 concurrencyOverrides {maxCount=3}
 concurrencySet [/Game/Sounds/Concurrency_Rules/Athena_EmoteFoley_Conc]
 soundSubmixSends [{minSendLevel=0, soundSubmix=&Fortnite_AudioAnalysis_Music}]
}
