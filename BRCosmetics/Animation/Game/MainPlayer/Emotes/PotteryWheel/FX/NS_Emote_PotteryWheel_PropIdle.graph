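# Niagara idle FX for the pottery wheel emote prop: electrical sparks and a warm flickering glow.
# ElectricalSparksBurst picks random wire sockets on the attach-parent mesh, emitting sprites and lights with gravity, drag and a sound cue.
# Spawn groups split moving sparks from a second, probability-based burst without launch speed or gravity.
# Glow holds two looping socket-attached sprites; brightness adapts to front-end versus gameplay, with time-of-day scaling on the glow.
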NiagaraSystem /BRCosmetics/Animation/Game/MainPlayer/Emotes/PotteryWheel/FX/NS_Emote_PotteryWheel_PropIdle
dependencies {
 usedDataInterfaces [
  &NiagaraDataInterfaceVectorCurve
  &NiagaraDataInterfaceAudioPlayer
  &NiagaraDataInterfaceCurve
  &NiagaraDataInterfaceSkeletalMesh
  &NiagaraDataInterfaceColorCurve
 ]
 usedDynamicInputs [&ColorFromCurve]
 usedModules [
  &self:SystemSpawnScript.SystemScriptSource.SystemScriptGraph.NiagaraNodeAssignment_2.SetVariables_#1
  &SystemState
  &self:SystemSpawnScript.SystemScriptSource.SystemScriptGraph.NiagaraNodeAssignment_1.SetVariables_#2
  &self:SystemSpawnScript.SystemScriptSource.SystemScriptGraph.NiagaraNodeAssignment_0.SetVariables_#3
 ]
}
emitters {
 ElectricalSparksBurst {
  emitterSpawnScript [&InitializeAudioPlayer(Audio Player=DataInterface() {propertyValues {
      Attenuation=null
      Concurrency=null
      MaxPlaysPerTick=10
      SoundToPlay=/BRCosmetics/Sounds/Emotes/PotteryWheel/EmoteFoley_PotteryWheel_Sparks_Cue
      bAllowLoopingOneShotSounds=false
      bLimitPlaysPerTick=true
      bOnlyActiveDuringGameplay=false
      bStopWhenComponentIsDestroyed=true
      ConfigurationUserParameter {parameter {name=None, typeDefHandle {registeredTypeIndex=1}, varData []}}
      ParameterNames []
     }})]
  emitterUpdateScript [
   /SetVariables(Emitter.RandomSockets:NiagaraInt32=&RandomRangeInt(Minimum=0, Maximum=4)) {isSetParameters=true}
   &EmitterState()
   &SpawnBurst_Instantaneous(
    Spawn Count=&RandomRangeInt(Minimum=7, Maximum=13)
    Use Spawn Probability=false
    UseLoopCountLimit=false
   )
   &SpawnBurst_Instantaneous(Spawn Group=1, Use Spawn Probability=true, UseLoopCountLimit=false) {module=SpawnBurst_Instantaneous001}
  ]
  particleSpawnScript [
   /SetVariables(Particles.VisibilityTag=$Emitter.SpawnGroup) {isSetParameters=true}
   /SetVariables(Particles.PerParticleRandom:NiagaraFloat=&FloatFromCurve(
     FloatCurve=DataInterface() {propertyValues {Curve {keys [
         {arriveTangent=4.1785, interpMode=RCIM_Cubic, tangentMode=RCTM_Break, time=0, value=0}
         {arriveTangent=3.1227, interpMode=RCIM_Cubic, leaveTangent=0.0224, tangentMode=RCTM_Break, time=1, value=1}
        ]}}}
     CurveIndex=&RandomRangeFloat(Minimum=0, Maximum=1)
    )) {isSetParameters=true}
   &InitializeParticle(
    Color Mode=::NewEnumerator0
    Lifetime=&RandomRangeFloat(Minimum=0.5, Maximum=1)
    Position Offset Coordinate Space=::Local
    UsePositionOffset=true
    Write Material Random=true
   )
   &SkeletalMeshLocation(
    Bone / Socket Index=$Emitter.RandomSockets
    Interpolate Mesh Sampling=false
    MaskSpawnGroup=false
    Mesh Sampling Type=::NewEnumerator5
    Skeletal Mesh=DataInterface() {propertyValues {
      DefaultMesh=null
      ExcludeBoneName=root
      OptionalComponentClass=null
      PreviewMesh=/BRCosmetics/Animation/Props/Gadgets_PotteryWheel/Meshes/PotteryWheel_Gadget
      SkinningMode=SkinOnTheFly
      SoftSourceActor=null
      SourceMode=AttachParent
      UvSetIndex=0
      WholeMeshLOD=-1
      bExcludeBone=false
      bReadDeformedGeometry=true
      bRequireCurrentFrameData=false
      ComponentTags []
      FilteredBones []
      FilteredSockets [JNT_WireSocket, JNT_WireSocket_0, JNT_WireSocket_1, JNT_WireSocket_2, JNT_WireSocket_3]
      MeshUserParameter {parameter {name=None, typeDefHandle {registeredTypeIndex=1}, varData []}}
      SamplingRegions []
     }}
    Socket Sampling Mode=::NewEnumerator5
   )
   &AddVelocity(
    Additional Quaternion Rotation=false
    Cone Angle=&RandomRangeFloat(Minimum=25, Maximum=90)
    Module.UseConeFalloff=false
    Module.UseVelDistribution=false
    Override Seed=false
    Rotation Mode=::NewEnumerator3
    Rotation Quaternion:Quat4f=$Particles.SkeletalMeshLocation.SampledRotation
    Velocity Mode=::NewEnumerator2
    Velocity Speed=&MaskFloatBySpawnGroup(Masked Float=&RandomRangeFloat(Minimum=150, Maximum=500), Passthrough Float=0)
   )
   &AddVelocity(
    Rotation Coordinate Space=::World
    Velocity=<0 0 1>
    Velocity Speed Scale=&RandomRangeFloat(Minimum=10, Maximum=40)
   ) {module=AddVelocity001}
   /SetVariables(
    Particles.FlareSpriteSize:Vector2f=<20 20>
    Particles.GlowSpriteSize:Vector2f=&Multiply_Vector2D_ByFloat(Float=1, Vector2D=&Lerp_Vector2(A=<5 10>, B=<10 20>, Alpha=$Particles.PerParticleRandom))
    Particles.SparkSpriteSize:Vector2f=&Multiply_Vector2D_ByFloat(Float=1, Vector2D=&Lerp_Vector2(A=<1 2>, B=<2 7>, Alpha=$Particles.PerParticleRandom))
   ) {isSetParameters=true}
   &CameraOffset(Camera Offset Amount=&MaskFloatBySpawnGroup(Masked Float=0, Passthrough Float=12))
   /Niagara/Modules/Audio/PlayAudio()
  ]
  particleUpdateScript [
   &SkeletalMeshLocation(
    Bone / Socket Index=$Emitter.RandomSockets
    MaskSpawnGroup=false
    Mesh Sampling Type=::NewEnumerator5
    Position Sampling=::NewEnumerator2
    Skeletal Mesh=DataInterface() {propertyValues {
      DefaultMesh=null
      ExcludeBoneName=root
      OptionalComponentClass=null
      PreviewMesh=&PotteryWheel_Gadget
      SkinningMode=SkinOnTheFly
      SoftSourceActor=null
      SourceMode=AttachParent
      UvSetIndex=0
      WholeMeshLOD=-1
      bExcludeBone=false
      bReadDeformedGeometry=true
      bRequireCurrentFrameData=false
      ComponentTags []
      FilteredBones []
      FilteredSockets [JNT_WireSocket, JNT_WireSocket_0, JNT_WireSocket_1, JNT_WireSocket_2, JNT_WireSocket_3]
      MeshUserParameter {parameter {name=None, typeDefHandle {registeredTypeIndex=1}, varData []}}
      SamplingRegions []
     }}
    Socket Sampling Mode=::NewEnumerator5
   ) {module=SkeletalMeshLocation001}
   /SetVariables(Particles.Position_LightGlow:NiagaraPosition=$Particles.SkeletalMeshLocation001.SampledPosition) {isSetParameters=true}
   /SetVariables(Particles.SparkSpriteColor:LinearColor=&MultiplyLinearColor(
     Linear Color=$System.SparkColorGradient
     Scale Factor=&MakeFloat4FromVectorAndFloat(
      "Vector (RGB)"=&VectorFromFloat(Value=&FN_FrontEndFloatModifier(Gameplay=1, Front End=0.4))
      "Float (Alpha)"=1
     )
    )) {isSetParameters=true}
   /SetVariables(Particles.GlowSpriteColor:LinearColor=&MultiplyLinearColor(
     Linear Color=$Particles.SparkSpriteColor
     Scale Factor=&MakeFloat4FromVectorAndFloat("Vector (RGB)"=<1 1 1>, "Float (Alpha)"=0.05)
    )) {isSetParameters=true}
   /SetVariables(Particles.FlareSpriteColor:LinearColor=&MultiplyLinearColor(
     Linear Color=$System.SparkColorGradient
     Scale Factor=&MakeFloat4FromVectorAndFloat(
      "Vector (RGB)"=&VectorFromFloat(Value=&FN_FrontEndFloatModifier(Gameplay=1, Front End=0.75))
      "Float (Alpha)"=&FloatFromCurve(
       FloatCurve=DataInterface() {propertyValues {Curve {keys [
           {tangentMode=RCTM_User, time=0, value=1}
           <at 0.0756 -0.7485>
           <at 0.1284 2.2279>
           <at 0.3098 -0.7559>
           <at 0.4525 0.9206>
           <at 0.525 0.0265>
           <at 0.6661 0.3676>
           <at 0.7436 1.6548>
           <at 0.8248 0.1441>
           <at 0.9286 0.8147>
           {tangentMode=RCTM_User, time=1, value=0}
          ]}}}
       CurveIndex=&NormalizeFloat(
        Value=&Multiply_Float(A=$Particles.NormalizedAge, B=&RandomRangeFloat(Minimum=2.5, Maximum=4))
        Min=0
        Max=1
       )
       Scale Curve=0.1
      )
     )
    )) {isSetParameters=true}
   /SetVariables(Particles.LightColor=&MultiplyLinearColor(
     Linear Color=$Particles.FlareSpriteColor
     Scale Factor=&MakeFloat4FromVectorAndFloat("Vector (RGB)"=&VectorFromFloat(Value=3), "Float (Alpha)"=1)
    )) {isSetParameters=true}
   &ParticleState()
   &Drag(
    Drag=&FloatFromCurve(
     FloatCurve=DataInterface() {propertyValues {Curve {keys [{interpMode=RCIM_Cubic, tangentMode=RCTM_User, time=0, value=1}, {interpMode=RCIM_Cubic, tangentMode=RCTM_User, time=1, value=0}]}}}
     Scale Curve=&Lerp_Float(A=17, B=5, Alpha=$Particles.PerParticleRandom)
    )
    Use Linear Drag=true
    Use Rotational Drag=true
   )
   &GravityForce(Gravity=&MaskVectorBySpawnGroup(Masked Vector=<0 0 -980>, Passthrough Vector=<0 0 0>))
   &InheritSourceMovement(
    Applied Position Scale=&VectorFromCurve(VectorCurve=DataInterface() {propertyValues {
       XCurve {keys [
         {arriveTangent=0.1831, interpMode=RCIM_Cubic, leaveTangent=0.1831, tangentMode=RCTM_User, time=0, value=1}
         {arriveTangent=-0.6074, interpMode=RCIM_Cubic, leaveTangent=-0.6074, tangentMode=RCTM_User, time=0.2, value=0.3}
         {arriveTangent=-0.2088, interpMode=RCIM_Cubic, leaveTangent=-0.2088, tangentMode=RCTM_User, time=1, value=0}
        ]}
       YCurve {keys [
         {arriveTangent=0.1831, interpMode=RCIM_Cubic, leaveTangent=0.1831, tangentMode=RCTM_User, time=0, value=1}
         {arriveTangent=-0.6074, interpMode=RCIM_Cubic, leaveTangent=-0.6074, tangentMode=RCTM_User, time=0.2, value=0.3}
         {arriveTangent=-0.2088, interpMode=RCIM_Cubic, leaveTangent=-0.2088, tangentMode=RCTM_User, time=1, value=0}
        ]}
       ZCurve {keys [
         {arriveTangent=0.1831, interpMode=RCIM_Cubic, leaveTangent=0.1831, tangentMode=RCTM_User, time=0, value=1}
         {arriveTangent=-0.6074, interpMode=RCIM_Cubic, leaveTangent=-0.6074, tangentMode=RCTM_User, time=0.2, value=0.3}
         {arriveTangent=-0.2088, interpMode=RCIM_Cubic, leaveTangent=-0.2088, tangentMode=RCTM_User, time=1, value=0}
        ]}
      }})
    ApplyForce=false
    ApplyPosition=true
    ApplyPositionFalloff=false
   )
   &SolveForcesAndVelocity(Clamp Velocity=false, Limit Acceleration=false)
  ]
  rendererClasses [&NiagaraSpriteRendererProperties, &NiagaraLightRendererProperties]
  renderers [
   {rendererClass=&NiagaraSpriteRendererProperties}
   {rendererClass=&NiagaraSpriteRendererProperties, rendererIndex=1}
   {rendererClass=&NiagaraLightRendererProperties, rendererIndex=2}
   {rendererClass=&NiagaraSpriteRendererProperties, rendererIndex=3}
  ]
 }
 Glow {
  emitterUpdateScript [
   &EmitterState(
    Inactive Response=::NewEnumerator1
    Life Cycle Mode=::NewEnumerator1
    Loop Behavior=::NewEnumerator1
    Loop Duration Mode=::NewEnumerator1
    UseLoopDelay=false
   )
   &SpawnBurst_Instantaneous(Spawn Count=2, Use Spawn Probability=false, UseLoopCountLimit=false)
  ]
  particleSpawnScript [
   &FN_SetSpawnGroup(Number of Spawn Groups=2)
   &InitializeParticle(Color=<rgba 1 0.5268 0 0.2>, Sprite Size Mode=::NewEnumerator3, Uniform Sprite Size=120, UsePositionOffset=false, Write Material Random=true)
   &CameraOffset(Camera Offset Amount=15)
  ]
  particleUpdateScript [
   &ParticleState(
    Kill Particles When Lifetime Has Elapsed=false
    Let Infinitely Lived Particles Die When Emitter Deactivates=true
    Loop Particles Lifetime=true
   )
   &SkeletalMeshLocation(
    Bone / Socket Index=$Particles.SpawnGroup
    MaskSpawnGroup=false
    Mesh Sampling Type=::NewEnumerator5
    Skeletal Mesh=$System.Skeletal Mesh
    Socket Sampling Mode=::NewEnumerator5
   )
   &ScaleColor(
    Scale Alpha=&FN_FrontEndFloatModifier(
     Gameplay=&FNTimeOfDayScalar(Night Scale=1.5, Morning Scale=1.75, Day Scale=2, Evening Scale=1.8)
     Front End=0.75
    )
    Scale RGB=&RandomRangeVector(
     Minimum=&VectorFromCurve(
      VectorCurve=DataInterface() {propertyValues {
        XCurve {keys [
          {interpMode=RCIM_Cubic, tangentMode=RCTM_User, time=0, value=1}
          {interpMode=RCIM_Cubic, tangentMode=RCTM_User, time=0.5, value=1.5}
          {tangentMode=RCTM_User, time=1, value=1}
         ]}
        YCurve {keys [
          {interpMode=RCIM_Cubic, tangentMode=RCTM_User, time=0, value=1}
          {interpMode=RCIM_Cubic, tangentMode=RCTM_User, time=0.5, value=1.5}
          {tangentMode=RCTM_User, time=1, value=1}
         ]}
        ZCurve {keys [
          {interpMode=RCIM_Cubic, tangentMode=RCTM_User, time=0, value=1}
          {interpMode=RCIM_Cubic, tangentMode=RCTM_User, time=0.5, value=1.5}
          {tangentMode=RCTM_User, time=1, value=1}
         ]}
       }}
      Scale Curve=<0.8 0.8 0.8>
     )
     Maximum=&VectorFromCurve(
      VectorCurve=DataInterface() {propertyValues {
        XCurve {keys [
          {interpMode=RCIM_Cubic, tangentMode=RCTM_User, time=0, value=1}
          {interpMode=RCIM_Cubic, tangentMode=RCTM_User, time=0.3, value=0.5}
          {arriveTangent=0.7143, interpMode=RCIM_Cubic, leaveTangent=0.7143, time=0.6984, value=1.3574}
          {tangentMode=RCTM_User, time=1, value=1}
         ]}
        YCurve {keys [
          {interpMode=RCIM_Cubic, tangentMode=RCTM_User, time=0, value=1}
          {interpMode=RCIM_Cubic, tangentMode=RCTM_User, time=0.3, value=0.5}
          {arriveTangent=0.7143, interpMode=RCIM_Cubic, leaveTangent=0.7143, time=0.695, value=1.3633}
          {tangentMode=RCTM_User, time=1, value=1}
         ]}
        ZCurve {keys [
          {interpMode=RCIM_Cubic, tangentMode=RCTM_User, time=0, value=1}
          {interpMode=RCIM_Cubic, tangentMode=RCTM_User, time=0.3, value=0.5}
          {arriveTangent=0.7143, interpMode=RCIM_Cubic, leaveTangent=0.7143, time=0.695, value=1.3633}
          {tangentMode=RCTM_User, time=1, value=1}
         ]}
       }}
      Scale Curve=<0.8 0.8 0.8>
     )
    )
    ScaleA=true
    ScaleRGB=true
   )
  ]
  rendererClasses [&NiagaraSpriteRendererProperties]
  renderers [{rendererClass=&NiagaraSpriteRendererProperties}]
 }
}
summary {
 systemName=NS_Emote_PotteryWheel_PropIdle
 emitters [
  {emitterName=Glow, rendererClasses [&NiagaraSpriteRendererProperties]}
  {
   emitterName=ElectricalSparksBurst
   rendererClasses [&NiagaraSpriteRendererProperties, &NiagaraLightRendererProperties]
  }
 ]
 userVariables [
  User.bIsFrontEnd {type=NiagaraBool}
  User.bIsFrontEndPreview {type=NiagaraBool}
 ]
}
